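# Niagara system for the LimeRock boss roar effect, with a User.GroundHeight vector parameter.
# DirectionBurst and MeshRipples1 are enabled but list no modules or renderers here; DistortionBlur, PP and SoftAirflow are disabled.
# DistortionBlur bursts a large camera-offset sprite with curve-driven dynamic material parameters, fading alpha and a component renderer.
# PP uses only a component renderer, scaling a loop-age intensity curve down to zero with distance.
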NiagaraSystem /LimeRock/Assets/FX/Boss_Roar/NS_LimeRock_Boss_Roar
dependencies {
 usedDataInterfaces [&NiagaraDataInterfaceCurve]
 usedDynamicInputs [
  &SubtractVectorFromPosition
  /Niagara/DynamicInputs/Helpers/SimulationPosition
  &Multiply_Vector
 ]
 usedModules [
  &self:SystemSpawnScript.SystemScriptSource.SystemScriptGraph.NiagaraNodeAssignment_0.SetVariables_#1
  &self:SystemSpawnScript.SystemScriptSource.SystemScriptGraph.NiagaraNodeAssignment_2.SetVariables_#2
  &SystemState
 ]
}
emitters {
 DirectionBurst {rendererClasses [], renderers []}
 DistortionBlur {
  bEnabled=false
  emitterUpdateScript [
   &EmitterState()
   &SpawnBurst_Instantaneous(Use Spawn Probability=false, UseLoopCountLimit=true)
  ]
  particleSpawnScript [
   &InitializeParticle(
    Lifetime=$System.CurrentLoopDuration
    Mesh Scale Mode=::NewEnumerator3
    Sprite Size Mode=::NewEnumerator3
    Uniform Sprite Size=1400
    UsePositionOffset=true
    Write Material Random=true
   )
   &CameraOffset(Camera Offset Amount=100)
  ]
  particleUpdateScript [
   &ParticleState()
   &Timeline(Play Rate=2, Timeline End=1.0000000200408773e+20)
   /SetVariables(Particles.float:NiagaraFloat=&FloatFromCurve(FloatCurve=DataInterface() {propertyValues {Curve {keys [
         {arriveTangent=8.4865, interpMode=RCIM_Cubic, time=0, value=0}
         {arriveTangent=1, interpMode=RCIM_Cubic, leaveTangent=1, time=0.25, value=0.9847}
         {interpMode=RCIM_Cubic, leaveTangent=-3.562, time=1, value=1}
        ]}}})) {isSetParameters=true}
   &DynamicMaterialParameters(
    Index 0 Param 1=$Particles.float
    Index 0 Param 2=$Particles.float
    Index 0 Param 3=0.8
    Index 1 Param 1=0.2
    Index 1 Param 2=500
    Index 1 Param 3=0.2
    Index 2 Param 1=$Particles.Timeline.Time
    Index 2 Param 2=$Particles.Timeline.Time
    Index 2 Param 3=$Particles.Timeline.Time
    Index 2 Param 4=0.5
    Index 3 Param 1=2
    Index 3 Param 2=200
    Module.Index0Param0Write=true
    Module.Index0Param1Write=true
    Module.Index0Param2Write=true
    Module.Index0Param3Write=true
    Module.Index1Param0Write=true
    Module.Index1Param1Write=true
    Module.Index1Param2Write=true
    Module.Index1Param3Write=true
    Module.Index2Param0Write=true
    Module.Index2Param1Write=true
    Module.Index2Param2Write=true
    Module.Index2Param3Write=true
    Module.Index3Param0Write=true
    Module.Index3Param1Write=true
    Module.Index3Param2Write=true
    Module.Index3Param3Write=true
    Write Parameter Index 1=true
    Write Parameter Index 2=true
    Write Parameter Index 3=true
   )
   &ScaleColor(
    Scale Alpha=&Multiply_Float(A=1, B=&FloatFromCurve(FloatCurve=DataInterface() {propertyValues {Curve {keys [<at 0 1>,<at 1 0>]}}}))
    Scale RGB=&Multiply_VectorByFloat(
     Vector=&VectorFromFloat(Value=&Waveform("Waveform [1]"=::NewEnumerator12, "[1] Frequency"=1, "[1] Amplitude Min/Max"=<1 1>))
     Float=&FloatFromCurve(
      FloatCurve=DataInterface() {propertyValues {Curve {keys [<at 0 1>, <at 0.2 1>, {arriveTangent=0.5082, time=1, value=1}]}}}
      CurveIndex=$Particles.float
     )
    )
    ScaleA=true
    ScaleRGB=true
   )
   /SetVariables(Particles.SpeedNormalizedB:NiagaraFloat=&FloatFromCurve(
     FloatCurve=DataInterface() {propertyValues {Curve {keys [
         <at 0 0 RCIM_Cubic>
         {arriveTangent=0.7355, interpMode=RCIM_Cubic, leaveTangent=0.7355, time=0.15, value=1}
         {arriveTangent=-1.1765, interpMode=RCIM_Cubic, leaveTangent=-1.1765, time=0.7892, value=0.5804}
         <at 1 0 RCIM_Cubic>
        ]}}}
     CurveIndex=$System.NormalizedLoopAge
     Scale Curve=0.677
    )) {isSetParameters=true}
  ]
  rendererClasses [&NiagaraSpriteRendererProperties, &NiagaraComponentRendererProperties]
  renderers [
   {rendererClass=&NiagaraSpriteRendererProperties}
   {rendererClass=&NiagaraComponentRendererProperties, rendererIndex=1}
  ]
 }
 MeshRipples1 {rendererClasses [], renderers []}
 PP {
  bEnabled=false
  emitterUpdateScript [
   &EmitterState()
   &SpawnBurst_Instantaneous(Use Spawn Probability=false, UseLoopCountLimit=true)
   &SpawnBurst_Instantaneous(Spawn Time=3.1167, Use Spawn Probability=false, UseLoopCountLimit=false) {module=SpawnBurst_Instantaneous001}
  ]
  particleSpawnScript [&InitializeParticle(
    Color Mode=::NewEnumerator0
    Lifetime=$System.CurrentLoopDuration
    Sprite Size Mode=::NewEnumerator3
    Uniform Sprite Size=1
    UsePositionOffset=false
    Write Material Random=true
   )]
  particleUpdateScript [
   /SetVariables(Particles.SpeedNormalizedB:NiagaraFloat=&FloatFromCurve(
     FloatCurve=DataInterface() {propertyValues {Curve {keys [
         <at 0 0 RCIM_Cubic>
         {arriveTangent=0.7355, interpMode=RCIM_Cubic, leaveTangent=0.7355, time=0.15, value=1}
         {arriveTangent=-1.1765, interpMode=RCIM_Cubic, leaveTangent=-1.1765, time=0.7892, value=0.5804}
         <at 1 0 RCIM_Cubic>
        ]}}}
     CurveIndex=$System.NormalizedLoopAge
     Scale Curve=&Multiply_Float(
      A=0.7
      B=&FN_LOD_Distance_Remap(Start Distance=12000, End Distance=14000, Minimum Output=1, Maximum Output=0, Clamp Output=true)
     )
    )) {isSetParameters=true}
   &ParticleState()
  ]
  rendererClasses [&NiagaraComponentRendererProperties]
  renderers [{rendererClass=&NiagaraComponentRendererProperties, rendererIndex=0}]
 }
 SoftAirflow {bEnabled=false, rendererClasses [], renderers []}
}
summary {
 systemName=NS_LimeRock_Boss_Roar
 emitters [
  {
   bEnabled=false
   emitterName=DistortionBlur
   rendererClasses [&NiagaraSpriteRendererProperties, &NiagaraComponentRendererProperties]
  }
  {bEnabled=false, emitterName=PP, rendererClasses [&NiagaraComponentRendererProperties]}
  {bEnabled=false, emitterName=SoftAirflow, rendererClasses []}
  {emitterName=MeshRipples1, rendererClasses []}
  {emitterName=DirectionBurst, rendererClasses []}
 ]
 userVariables [User.GroundHeight {type=Vector3f, description="", defaultValue {struct=Vector3f, value=<0 0 1300>}}]
}
